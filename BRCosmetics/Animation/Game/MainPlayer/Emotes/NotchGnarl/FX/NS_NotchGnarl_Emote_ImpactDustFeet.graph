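# Niagara one-shot burst effect for foot impacts in the NotchGnarl emote.
# Dust puffs rise in a cone and grow while fading, alongside brief mesh impacts that widen along X and very short streak sprites.
# SubUV debris bits fly out, then fall under gravity and drag with decaying spin.
# Dust and bits inherit negated velocity; curve-driven dynamic material parameters animate the dust and impact shaders.
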NiagaraSystem /BRCosmetics/Animation/Game/MainPlayer/Emotes/NotchGnarl/FX/NS_NotchGnarl_Emote_ImpactDustFeet
dependencies {
 usedDataInterfaces [&NiagaraDataInterfaceCurve, &NiagaraDataInterfaceColorCurve, &NiagaraDataInterfaceVector2DCurve]
 usedDynamicInputs []
 usedModules [&SystemState]
}
emitters {
 Dust {
  emitterUpdateScript [
   &EmitterState()
   &SpawnBurst_Instantaneous(Spawn Count=3, Use Spawn Probability=false, UseLoopCountLimit=false)
  ]
  particleSpawnScript [
   &InitializeParticle(
    Color Maximum=<rgba 0.5521 0.5291 0.5291 0.15>
    Color Minimum=<rgba 0.474 0.4396 0.4122 0.1>
    Color Mode=::NewEnumerator2
    Lifetime Max=0.5
    Lifetime Min=0.25
    Lifetime Mode=::NewEnumerator1
    Sprite Rotation Angle=&RandomRangeFloat(Minimum=0, Maximum=1)
    Sprite Rotation Mode=::NewEnumerator1
    Sprite Size Mode=::NewEnumerator1
    Uniform Sprite Size Max=35
    Uniform Sprite Size Min=25
    UsePositionOffset=false
    Write Material Random=true
   )
   &AddVelocity(
    Cone Angle=15
    Cone Axis=<0 0 1>
    Module.UseConeFalloff=false
    Module.UseVelDistribution=false
    Override Seed=false
    Rotation Coordinate Space=::World
    Velocity Mode=::NewEnumerator2
    Velocity Speed=&RandomRangeFloat(Minimum=100, Maximum=300)
   )
   &InheritVelocity(
    Inherited Velocity Amount Scale=&Multiply_VectorByFloat(
     Vector=<-1 -1 -1>
     Float=&FloatFromCurve(FloatCurve=DataInterface() {propertyValues {Curve {keys [<at 0 1>,<at 1 0>]}}})
    )
    Speed Limit=true
    UseSpeedThreshold=false
   )
   &ShapeLocation(MaskSpawnGroup=false, Override Seed=false, Sphere Radius=15, UseOwnerScale=true)
   &ApplyInitialForces()
  ]
  particleUpdateScript [
   &Drag(Drag=3, Use Linear Drag=true, Use Rotational Drag=false)
   &SolveForcesAndVelocity(Clamp Velocity=false, Limit Acceleration=false)
   &ParticleState()
   &DynamicMaterialParameters(
    Index 0 Param 1=&FloatFromCurve(FloatCurve=DataInterface() {propertyValues {Curve {keys [<at 0 0>,<at 1 1>]}}})
    Index 0 Param 2=1
    Index 0 Param 4=1
    Index 2 Param 1=&FloatFromCurve(FloatCurve=DataInterface() {propertyValues {Curve {keys [
         {arriveTangent=3.7909, interpMode=RCIM_Cubic, leaveTangent=0.7255, tangentMode=RCTM_Break, time=0, value=0}
         {arriveTangent=0.558, interpMode=RCIM_Cubic, leaveTangent=0.4094, tangentMode=RCTM_Break, time=0.2, value=0.12}
         {arriveTangent=-0.0287, interpMode=RCIM_Cubic, leaveTangent=0.0797, tangentMode=RCTM_Break, time=1, value=0.2}
        ]}}})
    Index 2 Param 2=0.4
    Index 2 Param 3=&FloatFromCurve(FloatCurve=DataInterface() {propertyValues {Curve {keys [
         {arriveTangent=3.7909, interpMode=RCIM_Cubic, leaveTangent=3.7909, tangentMode=RCTM_User, time=0, value=0}
         {arriveTangent=1.8782, interpMode=RCIM_Cubic, leaveTangent=1.8782, tangentMode=RCTM_User, time=0.2, value=0.55}
         {arriveTangent=0.0797, interpMode=RCIM_Cubic, leaveTangent=0.0797, tangentMode=RCTM_User, time=1, value=1}
        ]}}})
    Index 2 Param 4=&FloatFromCurve(FloatCurve=DataInterface() {propertyValues {Curve {keys [<at 0 0>,<at 1 2>]}}})
    Module.Param0WriteEnabled=true
    Module.Param0WriteEnabled2=true
    Module.Param1WriteEnabled=true
    Module.Param1WriteEnabled2=true
    Module.Param2WriteEnabled=true
    Module.Param2WriteEnabled2=true
    Module.Param3WriteEnabled=true
    Module.Param3WriteEnabled2=true
    Write Parameter Index 2=true
   )
   &ScaleColor(
    Scale Alpha=&FloatFromCurve(FloatCurve=DataInterface() {propertyValues {Curve {keys [
         {interpMode=RCIM_Cubic, leaveTangent=9.8397, tangentMode=RCTM_Break, time=0, value=0}
         {interpMode=RCIM_Cubic, tangentMode=RCTM_Break, time=0.2, value=1}
         {arriveTangent=0.2181, interpMode=RCIM_Cubic, tangentMode=RCTM_Break, time=1, value=0}
        ]}}})
    ScaleA=true
    ScaleRGB=true
   )
   &ScaleSpriteSize(Uniform Curve Sprite Scale=DataInterface() {propertyValues {Curve {keys [
        {arriveTangent=1.7258, interpMode=RCIM_Cubic, leaveTangent=1.7258, tangentMode=RCTM_User, time=0, value=1}
        {arriveTangent=1.1259, interpMode=RCIM_Cubic, leaveTangent=1.1259, tangentMode=RCTM_User, time=0.2, value=1.45}
        {arriveTangent=0.0885, interpMode=RCIM_Cubic, leaveTangent=0.0885, tangentMode=RCTM_User, time=1, value=1.7}
       ]}}})
  ]
  rendererClasses [&NiagaraSpriteRendererProperties]
  renderers [
   {rendererClass=&NiagaraSpriteRendererProperties}
   {rendererClass=&NiagaraSpriteRendererProperties, rendererIndex=1}
  ]
 }
 Impact {
  emitterUpdateScript [
   &EmitterState()
   &SpawnBurst_Instantaneous(Spawn Count=3, Use Spawn Probability=false, UseLoopCountLimit=false)
  ]
  particleSpawnScript [
   /SetVariables(Particles.SplineWidth=2) {isSetParameters=true}
   &InitializeParticle(
    Color=<rgba 1 0.8647 0.8125 1>
    Lifetime Max=0.22
    Lifetime Min=0.11
    Lifetime Mode=::NewEnumerator1
    Mesh Scale Mode=::NewEnumerator1
    Mesh Uniform Scale Max=0.7
    Mesh Uniform Scale Min=0.5
    UsePositionOffset=false
    Write Material Random=true
   )
   &AddVelocity(Module.Use Falloff Distance=false, Origin Offset=<0 0 -1>, Override Seed=false, Velocity Mode=::NewEnumerator1) {module=AddVelocity001}
   &InheritVelocity(Speed Limit=true, UseSpeedThreshold=false)
   &ApplyInitialForces()
   &InitialMeshOrientation(
    Mesh Orientation Mode=::NewEnumerator2
    Orientation Vector=<0 0 1>
    Rotation=&RandomRangeVector(Minimum=<-0.15 -0.15 -0.15>, Maximum=<0.15 0.15 0.15>)
    Use Rotation Vector=true
   )
   &ShapeLocation(
    Cylinder Height=0
    Cylinder Radius=10
    MaskSpawnGroup=false
    Override Seed=false
    Shape Primitive=::NewEnumerator1
    Surface Only=false
    UseOwnerScale=true
   )
  ]
  particleUpdateScript [
   &DynamicMaterialParameters(
    Index 0 Param 1=&FloatFromCurve(FloatCurve=DataInterface() {propertyValues {Curve {keys [{arriveTangent=2.9529, interpMode=RCIM_Cubic, leaveTangent=2.9529, tangentMode=RCTM_User, time=0, value=0}, <at 1 1>]}}})
    Index 0 Param 2=2
    Index 0 Param 3=$Particles.MaterialRandom
    Index 0 Param 4=&FloatFromCurve(FloatCurve=DataInterface() {propertyValues {Curve {keys [<at 0 0.5>,<at 1 1>]}}}, Scale Curve=$Particles.SplineWidth)
    Module.Param0WriteEnabled=true
    Module.Param1WriteEnabled=true
    Module.Param2WriteEnabled=true
    Module.Param3WriteEnabled=true
   )
   &ScaleColor(
    Scale Alpha=&FloatFromCurve(FloatCurve=DataInterface() {propertyValues {Curve {keys [
         {arriveTangent=0.002, interpMode=RCIM_Cubic, time=0, value=0}
         {arriveTangent=0.002, interpMode=RCIM_Cubic, leaveTangent=0.002, time=0.2, value=1}
         {arriveTangent=0.1932, interpMode=RCIM_Cubic, leaveTangent=0.1932, tangentMode=RCTM_User, time=1.0011, value=0.002}
        ]}}})
    ScaleA=true
    ScaleRGB=true
   )
   &ParticleState()
   &ScaleMeshSize(Scale Factor=&MakeVector(
     X=&FloatFromCurve(
      FloatCurve=DataInterface() {propertyValues {Curve {keys [{arriveTangent=2.3365, interpMode=RCIM_Cubic, leaveTangent=2.3365, tangentMode=RCTM_User, time=0, value=0}, <at 1 1>]}}}
      Scale Curve=1.25
     )
     Y=1
     Z=1
    ))
   &SolveForcesAndVelocity(Clamp Velocity=false, Limit Acceleration=false)
  ]
  rendererClasses [&NiagaraMeshRendererProperties]
  renderers [{rendererClass=&NiagaraMeshRendererProperties, rendererIndex=0}]
 }
 Lines {
  emitterUpdateScript [
   &EmitterState()
   &SpawnBurst_Instantaneous(Spawn Count=6, Use Spawn Probability=false, UseLoopCountLimit=false)
  ]
  particleSpawnScript [
   &InitializeParticle(
    Color Mode=::NewEnumerator0
    Lifetime Max=0.07
    Lifetime Min=0.05
    Lifetime Mode=::NewEnumerator1
    Sprite Size Max=<2 30>
    Sprite Size Min=<2 10>
    Sprite Size Mode=::NewEnumerator2
    UsePositionOffset=false
    Write Material Random=true
   )
   &DynamicMaterialParameters(
    Index 0 Param 1=2
    Index 0 Param 2=10
    Index 0 Param 3=1
    Module.Param0WriteEnabled=true
    Module.Param0WriteEnabled1=true
    Module.Param1WriteEnabled=true
    Module.Param1WriteEnabled1=true
    Module.Param2WriteEnabled=true
    Module.Param2WriteEnabled1=true
    Module.Param3WriteEnabled=true
    Module.Param3WriteEnabled1=true
    Write Parameter Index 1=true
   )
   &ShapeLocation(
    MaskSpawnGroup=false
    Offset=<0 0 15>
    Offset Coordinate Space=::Local
    Offset Mode=::NewEnumerator0
    Override Seed=false
    Sphere Radius=20
    Sphere Surface Distribution=1
    UseOwnerScale=true
   )
   &AddVelocity(Module.Use Falloff Distance=false, Origin Offset=<0 0 -10>, Override Seed=false, Velocity Mode=::NewEnumerator1, Velocity Speed=350)
   &ApplyInitialForces()
  ]
  particleUpdateScript [
   &ParticleState()
   &Drag(Drag=2, Use Linear Drag=true, Use Rotational Drag=true)
   &SolveForcesAndVelocity(Clamp Velocity=false, Limit Acceleration=false)
  ]
  rendererClasses [&NiagaraSpriteRendererProperties]
  renderers [{rendererClass=&NiagaraSpriteRendererProperties}]
 }
 SmallBits {
  emitterUpdateScript [
   &EmitterState()
   &SpawnBurst_Instantaneous(
    Spawn Count=&RandomRangeInt(Minimum=5, Maximum=10)
    Use Spawn Probability=false
    UseLoopCountLimit=false
   )
  ]
  particleSpawnScript [
   /SetVariables(Particles.randomInt:NiagaraInt32=&RandomRangeInt(Minimum=0, Maximum=1)) {isSetParameters=true}
   &InitializeParticle(
    Color Maximum=<rgba 0.526 0.5217 0.4789 5>
    Color Minimum=<rgba 0.2279 0.2292 0.2078 5>
    Color Mode=::NewEnumerator2
    Lifetime=&RandomRangeFloat(Minimum=1.5, Maximum=2)
    Mesh Scale Mode=::NewEnumerator1
    Mesh Uniform Scale Max=50
    Mesh Uniform Scale Min=12
    Sprite Rotation Mode=::NewEnumerator2
    Sprite Size Mode=::NewEnumerator1
    Uniform Sprite Size Max=4
    Uniform Sprite Size Min=3
    UsePositionOffset=false
    Write Material Random=true
   )
   &InheritVelocity(
    Inherited Velocity Amount Scale=&Multiply_VectorByFloat(
     Vector=<-1 -1 -1>
     Float=&FloatFromCurve(FloatCurve=DataInterface() {propertyValues {Curve {keys [<at 0 1>,<at 1 0>]}}})
    )
    Speed Limit=true
    UseSpeedThreshold=false
   )
   &SubUVAnimation(End Frame=3, SubUV Animation Mode:ENiagara_SubUVLookupMode=::NewEnumerator2)
   &ShapeLocation(MaskSpawnGroup=false, Override Seed=false, Sphere Radius=10, UseOwnerScale=true)
   /SetVariables(Particles.InitSpriteRotRate=&UniformRangedFloat(Minimum=-3, Maximum=3)) {isSetParameters=true}
   &AddVelocity(
    Module.Use Falloff Distance=false
    Origin Offset=<0 0 -1>
    Override Seed=false
    Velocity Mode=::NewEnumerator1
    Velocity Speed=&RandomRangeFloat(Minimum=350, Maximum=750)
   )
   &ApplyInitialForces()
  ]
  particleUpdateScript [
   &ParticleState()
   &SpriteRotationRate(Rotation Rate=&Multiply_Float(
     A=$Particles.InitSpriteRotRate
     B=&FloatFromCurve(
      FloatCurve=DataInterface() {propertyValues {Curve {keys [
          {arriveTangent=-0.2085, interpMode=RCIM_Cubic, leaveTangent=-0.2085, tangentMode=RCTM_User, time=-0.0046, value=1.0113}
          {arriveTangent=-0.2496, interpMode=RCIM_Cubic, leaveTangent=-0.2496, tangentMode=RCTM_User, time=0.5, value=0.05}
         ]}}}
      CurveIndex=$Emitter.Age
      Scale Curve=950
     )
    ))
   &GravityForce(Gravity=&RandomRangeVector(Minimum=<0 0 -980>, Maximum=<1 1 -1500>))
   &Drag(
    Drag=&FloatFromCurve(
     FloatCurve=DataInterface() {propertyValues {Curve {keys [{arriveTangent=-1, interpMode=RCIM_Cubic, leaveTangent=-3.4078, tangentMode=RCTM_Break, time=0, value=1}, <at 0.3 0.1>, <at 1 0>]}}}
     Scale Curve=7
    )
    Use Linear Drag=true
    Use Rotational Drag=false
   )
   &SolveForcesAndVelocity(Clamp Velocity=false, Limit Acceleration=false)
   &ScaleSpriteSize(
    Uniform Curve Scale=2.5
    Uniform Curve Sprite Scale=DataInterface() {propertyValues {Curve {keys [
        {arriveTangent=40, interpMode=RCIM_Cubic, leaveTangent=40, time=0, value=0}
        {arriveTangent=-2.1262, interpMode=RCIM_Cubic, leaveTangent=-2.1262, tangentMode=RCTM_User, time=0.025, value=1}
        <at 0.125 0.2857>
        <at 0.9 0.3>
        {arriveTangent=-0.0285, interpMode=RCIM_Cubic, leaveTangent=-0.0285, tangentMode=RCTM_User, time=1, value=0}
       ]}}}
   )
  ]
  rendererClasses [&NiagaraSpriteRendererProperties]
  renderers [{rendererClass=&NiagaraSpriteRendererProperties}]
 }
}
summary {
 systemName=NS_NotchGnarl_Emote_ImpactDustFeet
 emitters [
  {emitterName=Impact, rendererClasses [&NiagaraMeshRendererProperties]}
  {emitterName=Dust, rendererClasses [&NiagaraSpriteRendererProperties]}
  {emitterName=Lines, rendererClasses [&NiagaraSpriteRendererProperties]}
  {emitterName=SmallBits, rendererClasses [&NiagaraSpriteRendererProperties]}
 ]
 userVariables []
}
